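# Default Lit material for the demo drivable golf cart base, flagged for skeletal mesh, mesh deformer and Niagara mesh particle use.
# It has BaseColor and Normal texture parameters.
# An FX mask's red channel scales the BoostGlow color, black by default, into an emissive adjustment function.
# Camera-to-pixel distance, offset and divided by a cull distance parameter, drives a temporal AA dither threshold for distance-based fading.
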Material /Game/Athena/DrivableVehicles/Golf_Cart/Golf_Cart_Base/Demo/Materials/Demo_Golf_Cart_Base
graph {
 properties {
  bUsedWithMeshDeformer=true
  bUsedWithNiagaraMeshParticles=true
  bUsedWithSkeletalMesh=true
  pixelDepthOffsetMode=PDOM_Legacy
  usedShadingModels=Default Lit
 }
 nodes [
  Constant() {r=0.2}
  Constant() {r=0.5}
  n1 = TextureSample() {samplerType=SAMPLERTYPE_Masks, texture=/Game/Athena/DrivableVehicles/Golf_Cart/Textures/T_Golf_Cart_Base_Demo_Golf_Cart_Base_FX}
  n2 = VectorParameter() {parameterName=BoostGlow, defaultValue=<rgba 0 0 0 1>}
  MaterialFunctionCall(Emissive=Add(A=Multiply(A=@n1.R, B=@n2.RGB)) {constB=0}) {materialFunction=&MF_FX_EmissiveAdjustments}
  n3 = Distance(A=WorldPosition().XYZ, B=CameraPositionWS())
  n4 = ScalarParameter() {defaultValue=-20, parameterName=s}
  MaterialFunctionCall(Alpha Threshold=Divide(A=Add(A=@n3, B=@n4), B=ScalarParameter() {defaultValue=40, parameterName=cull distance}), Random=Constant()) {materialFunction=&DitherTemporalAA_Fortnite}
  TextureSampleParameter2D() {parameterName=BaseColor, texture=/Game/Athena/DrivableVehicles/Golf_Cart/Golf_Cart_Base/Demo/Textures/Demo_T_Golf_Cart_Base_D}
  TextureSampleParameter2D() {
   parameterName=Normal
   samplerType=SAMPLERTYPE_Normal
   texture=/Game/Athena/DrivableVehicles/Golf_Cart/Golf_Cart_Base/Demo/Textures/Demo_T_Golf_Cart_Base_N
  }
 ]
}
